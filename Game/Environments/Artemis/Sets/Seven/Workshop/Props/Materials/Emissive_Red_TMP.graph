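# Temporary red material, named as an emissive, in the Artemis Seven Workshop props set, using Default Lit shading.
# It holds a red vector parameter, a constant vector left at default, and scalar parameters for metalness, specular and roughness.
# The graph shows no links between these nodes or to material outputs.
# Usage flags cover Nanite, skeletal meshes and mesh deformers.
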Material /Game/Environments/Artemis/Sets/Seven/Workshop/Props/Materials/Emissive_Red_TMP
graph {
 properties {
  bUsedWithMeshDeformer=true
  bUsedWithNanite=true
  bUsedWithSkeletalMesh=true
  pixelDepthOffsetMode=PDOM_Legacy
  usedShadingModels=Default Lit
 }
 nodes [
  Constant3Vector()
  ScalarParameter() {defaultValue=1, parameterName=Metal}
  ScalarParameter() {parameterName=Spec}
  ScalarParameter() {parameterName=Rough}
  VectorParameter() {defaultValue=<rgba 0.6181 0.0013 0.0015 0>}
 ]
}
